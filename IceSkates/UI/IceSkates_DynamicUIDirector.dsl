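; Dynamic UI director that shows the Ice Skates UI scene while skates are in the player's inventory and the player is not in a vehicle.
; On BeginPlay and each possession change, it caches the player pawn and starts gameplay-tag wait tasks for the inventory and in-vehicle tags.
; These tasks set two flags that UpdateIceSkatesSceneState combines, and a delta gate adds or removes the scene only when the result changes.
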(event EventBeginPlay
  ; Begin/EndPlay
  (bind _output_get (Variables|Default|SetBoundPlayerController (DynamicUIDirector|GetOwningLocalPlayerController)))
  (Pawn|BindEventtoOnPossessedPawnChanged _output_get (EventDispatchers|CreateEvent :!event OnPossessedPawnChanged))
  (CallFunction|ListenforPawnEvents (Pawn|GetControlledPawn (Variables|Default|GetBoundPlayerController))))

(event EventEndPlay
  (params (EndPlayReason :type EEndPlayReason%20Enum :default "Destroyed"))
  (bind _boundplayercontroller (Variables|Default|GetBoundPlayerController))
  ; Begin/EndPlay
  (Utilities|IsValid _boundplayercontroller
    (:Is%20Valid
      (Pawn|UnbindEventfromOnPossessedPawnChanged _boundplayercontroller (EventDispatchers|CreateEvent :!event OnPossessedPawnChanged)))))

(event Custom|ListenForPawnEvents
  (params (Pawn :type Pawn%20Object%20Reference))
  ;[0 0 0] Listen For Pawn Events
  (bind _asfort_player_pawn (Utilities|Casting|CastToFortPlayerPawn Pawn))
  (Variables|Default|SetPlayerPawn _asfort_player_pawn)
  (CallFunction|ListenforIceSkatesinInventory)
  (CallFunction|ListenforVehicleState))

(event Custom|ListenForIceSkatesInInventory
  (bind _iceskatesininventorytag (Variables|IceSkates|GetIceSkatesInInventoryTag))
  ; Listen for Ice Skates in Inventory
  (Ability|Tasks|WaitGameplayTagAddToActor (Variables|Default|GetPlayerPawn) _iceskatesininventorytag
    (:then
      (Ability|Tasks|WaitGameplayTagRemoveFromActor (Variables|Default|GetPlayerPawn) _iceskatesininventorytag
        (:Removed
          (Variables|IceSkates|SetIsIceSkatesInInventory false))))
    (:Added
      (Variables|IceSkates|SetIsIceSkatesInInventory true)))
  (CallFunction|UpdateIceSkatesSceneState))

(event Custom|ListenForVehicleState
  ; ListenForVehicleState
  (Ability|Tasks|WaitGameplayTagAddToActor (Variables|Default|GetPlayerPawn) "(TagName=\"Gameplay.Status.InVehicle\")"
    (:then
      (Ability|Tasks|WaitGameplayTagRemoveFromActor (Variables|Default|GetPlayerPawn) "(TagName=\"Gameplay.Status.InVehicle\")"
        (:Removed
          (Variables|Default|SetIsInVehicle? false)
          (CallFunction|UpdateIceSkatesSceneState))))
    (:Added
      (Variables|Default|SetIsInVehicle? true)
      (CallFunction|UpdateIceSkatesSceneState))))

(fn OnPossessedPawnChanged
  (params (OldPawn :type Pawn%20Object%20Reference) (NewPawn :type Pawn%20Object%20Reference))
  (CallFunction|ListenforPawnEvents NewPawn))

(fn SetIceSkatesSceneActive
  (params (IsActive :type Boolean))
  (bind _scene_iceskates (Variables|IceSkates|GetScene_IceSkates))
  (bind _updated_variable (Utilities|FlowControl|DeltaGateWildcard (Variables|IceSkates|GetIsIceSkatesSceneActive) IsActive)
    (:Success
      (if _updated_variable
        (DynamicUIDirector|AddScene _scene_iceskates)
        (else
          (DynamicUIDirector|RemoveScene _scene_iceskates))))))

(fn UpdateIceSkatesSceneState
  (CallFunction|SetIceSkatesSceneActive (and (Variables|IceSkates|GetIsIceSkatesInInventory) (not (Variables|Default|GetIsInVehicle?)))))

(fn ConstructionScript)
