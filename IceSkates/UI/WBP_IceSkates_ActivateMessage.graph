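# Ice Skates HUD prompt widget that shows the Fire keybind beside an Activate/Deactivate label.
# On Construct, the ShouldShowWeaponHudMessage curve-table value collapses the widget when true, despite its name, and makes it visible when false.
# The visible path starts repeating async listeners for the IcyFeet Equipped gameplay tag on the owning pawn.
# Tag addition sets the label to Deactivate, and removal sets it to Activate.
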WidgetBlueprint /IceSkates/UI/WBP_IceSkates_ActivateMessage
graphs [EventGraph {nodes [
   Event Construct()
   > n1 = Branch(Condition=GetValueAsBool!(
     Input:"Scalable Float Structure (by ref)"=Get Row_ShouldShowWeaponHudMessage() -> (Row_ShouldShowWeaponHudMessage:Scalable Float Structure)
     Level:float=0
    ) -> (ReturnValue:bool))
   > SetVisibility(InVisibility=Collapsed)
   SetVisibility(execute=@n1.else, InVisibility=Visible)
   n2 = GetOwningPlayerPawn!(self:User Widget Object Reference) -> (ReturnValue:Pawn Object Reference)
   > n3 = WaitGameplayTagAddToActor!(
    execute:Exec
    TargetActor:Actor Object Reference=@n2
    Tag:Gameplay Tag Structure=Get IceSkates Active() -> (IceSkates Active:Gameplay Tag Structure)
    OnlyTriggerOnce:bool=false
   ) -> (then:Exec, AsyncTaskProxy:Ability Async Wait Gameplay Tag Added Object Reference, Added:Exec) {is=title K2Node_AsyncAction}
   > n4 = WaitGameplayTagRemoveFromActor!(execute:Exec, TargetActor:Actor Object Reference=@n2, Tag:Gameplay Tag Structure=Get IceSkates Active(), OnlyTriggerOnce:bool=false) -> (then:Exec, AsyncTaskProxy:Ability Async Wait Gameplay Tag Removed Object Reference, Removed:Exec) {is=title K2Node_AsyncAction}
   n5 = Get Text_State() -> (Text_State:Text Block Object Reference)
   Set Text(execute=@n4.Removed, Text=Get TextActivate() -> (TextActivate:Text), self=@n5)
   Set Text(execute=@n3.Added, Text=Get TextDeactivate() -> (TextDeactivate:Text), self=@n5)
   Event PreConstruct()
  ]}]
variables [
 IceSkates Active {type=GameplayTag, default {tagName=Gameplay.Status.IcyFeet.Equipped}}
 Row_ShouldShowWeaponHudMessage {
  type=ScalableFloat
  default {
   registryType=None
   value=1
   curve {curveTable=/IceSkates/DataTables/IceSkatesGameplayGameData, rowName=Default.IceSkates.GA.ShouldShowWeaponHudMessage}
  }
 }
 TextActivate {type=string, default=Activate}
 TextDeactivate {type=string, default=Deactivate}
]
widgets {
 info {inheritedWidgetCount=0, parentClass=&UserWidget, rootWidgetClass=&HorizontalBox, widgetCount=3}
 widgets [
  HorizontalBox_58 {widgetClassPath=&HorizontalBox, widget=HorizontalBox_58, properties {flowDirectionPreference=Culture}}
  KeybindWidget {
   widgetClassPath=/FortUILibrary/Input/KeybindWidget
   bIsVariable=true
   parent=HorizontalBox_58
   widget=KeybindWidget
   properties {boundAction=Fire, padding=<ltrb 10 10 10 10>}
   slot {class=HorizontalBoxSlot, properties {horizontalAlignment=HAlign_Center, verticalAlignment=VAlign_Center, padding=<ltrb -10 -10 -10 -10>}}
  }
  Text_State {
   widgetClassPath=&UEFN_TextBlock
   bIsVariable=true
   parent=HorizontalBox_58
   widget=Text_State
   properties {
    text=Activate/Deactivate
    font {
     bForceMonospaced=false
     bMaterialIsStencil=false
     fontMaterial=null
     fontObject=&BurbankSmall
     letterSpacing=0
     monospacedWidth=1
     size=18
     skewAmount=0
     typefaceFontName=Black
     outlineSettings {
      bApplyOutlineToDropShadows=false
      bMiteredCorners=false
      bSeparateFillAlpha=false
      outlineMaterial=null
      outlineSize=0
      outlineColor=<rgba 0 0 0 1>
     }
    }
   }
   slot {class=HorizontalBoxSlot, properties {horizontalAlignment=HAlign_Left, verticalAlignment=VAlign_Center, padding=<ltrb 10 0 0 0>}}
  }
 ]
}
